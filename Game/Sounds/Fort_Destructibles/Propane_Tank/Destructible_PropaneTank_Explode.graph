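# Sound cue for a propane tank destructible exploding, built from close, medium and distant layers summed in a mixer, with the medium layer quieter.
# Each layer has its own attenuation and sound class, with only the close one overriding its attenuation settings.
# The close layer plays a dedicated wave with varied pitch, while medium and distant randomly pick pitched-up, borrowed grenade and gas pump explosions.
# A concurrency override tightly caps simultaneous instances.
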SoundCue /Game/Sounds/Fort_Destructibles/Propane_Tank/Destructible_PropaneTank_Explode
root=Mixer_4
nodes [
 {
  id=Attenuation_0
  children [Modulator_2]
  properties {
   attenuationSettings=/Game/Sounds/Attenuation/Fort_Explosion_Distant_Athena_Att
   attenuationOverrides {falloffDistance=3000, attenuationShapeExtents {x=1000}}
  }
 }
 {
  id=Attenuation_4
  children [WavePlayer_0]
  properties {
   attenuationSettings=/Game/Sounds/Attenuation/Fort_Explosion_Close_Athena_Att
   bOverrideAttenuation=true
   attenuationOverrides {falloffDistance=5106.2798, stereoSpread=5084.1587, attenuationShapeExtents {x=2554.7803}}
  }
 }
 {
  id=Attenuation_5
  children [Modulator_5]
  properties {
   attenuationSettings=/Game/Sounds/Attenuation/Fort_Explosion_Medium_Athena_Att
   attenuationOverrides {falloffDistance=3000, attenuationShapeExtents {x=1000}}
  }
 }
 {id=Mixer_4, children [SoundClass_3,SoundClass_2,SoundClass_1], properties {inputVolume [1,0.75,1]}}
 {id=Modulator_1, children [Attenuation_4], properties {pitchMax=1.2}}
 {id=Modulator_2, children [Random_2], properties {pitchMax=1.3, pitchMin=1.2}}
 {id=Modulator_5, children [Random_3], properties {pitchMax=1.3, pitchMin=1.2}}
 {id=Random_2, children [WavePlayer_3,WavePlayer_4], properties {weights [1,1]}}
 {id=Random_3, children [WavePlayer_6,WavePlayer_7], properties {weights [1,1]}}
 {
  id=SoundClass_1
  children [Attenuation_0]
  properties {soundClassOverride=&Fort_WeaponRanged_Fire_Distant}
 }
 {
  id=SoundClass_2
  children [Attenuation_5]
  properties {soundClassOverride=&Fort_WeaponRanged_Fire_Distant}
 }
 {id=SoundClass_3, children [Modulator_1], properties {soundClassOverride=&Fort_WeaponRanged_Fire_3P_PSM}}
 {id=WavePlayer_0, properties {soundWaveAssetPtr=&/Destructible_Propane_Tank_Explode_01}}
 {id=WavePlayer_3, properties {soundWaveAssetPtr=/Game/Sounds/Fort_Destructibles/GasPump/Destructible_Gas_Pump_EXPLODE_Distant_01}}
 {id=WavePlayer_4, properties {soundWaveAssetPtr=/Game/Sounds/Fort_Destructibles/GasPump/Destructible_Gas_Pump_EXPLODE_Distant_02}}
 {id=WavePlayer_6, properties {soundWaveAssetPtr=/Game/Athena/Sounds/Weapons/Explosions/Grenade_Expl_Med_01}}
 {id=WavePlayer_7, properties {soundWaveAssetPtr=/Game/Athena/Sounds/Weapons/Explosions/Grenade_Expl_Med_02}}
]
properties {bOverrideConcurrency=true, maxDistance=27500, concurrencyOverrides {maxCount=2, volumeScale=0.8762}}
