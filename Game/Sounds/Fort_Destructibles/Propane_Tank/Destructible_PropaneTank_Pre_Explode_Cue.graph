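# Sound cue for the warning sound a destructible propane tank plays before it explodes.
# A single wave runs through a Fort_Gadgets sound class override and a modulator that only randomizes pitch upward, then into an attenuation node.
# That node references the RPG explosion attenuation asset but overrides it with custom falloff and distance-based low-pass filtering.
# A concurrency override tightly caps simultaneous instances, and the cue plays at reduced volume.
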SoundCue /Game/Sounds/Fort_Destructibles/Propane_Tank/Destructible_PropaneTank_Pre_Explode_Cue
root=Attenuation_0
nodes [
 {
  id=Attenuation_0
  children [Modulator_1]
  properties {
   attenuationSettings=/Game/Sounds/Attenuation/Fort_Explosion_RPG_Att
   bOverrideAttenuation=true
   attenuationOverrides {bAttenuateWithLPF=true, lPFFrequencyAtMax=665.3513, lPFRadiusMax=2786.9099, lPFRadiusMin=2024.5316, attenuationShapeExtents {x=1407.6522}}
  }
 }
 {id=Modulator_1, children [SoundClass_1], properties {pitchMax=1.3, pitchMin=1, volumeMax=1, volumeMin=1}}
 {id=SoundClass_1, children [WavePlayer_1], properties {soundClassOverride=&Fort_Gadgets}}
 {id=WavePlayer_1, properties {soundWaveAssetPtr=&/Destructible_PropaneTank_Pre_Explode}}
]
properties {bOverrideConcurrency=true, maxDistance=5007.6523, volumeMultiplier=0.5, concurrencyOverrides {maxCount=2, volumeScale=0.819}}
